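# Default Lit material for the button on the TechDoor in the GatedEntrance secondary mission.
# A vector parameter, defaulting to a bright cyan-blue, has its RGB multiplied by a constant to boost intensity beyond normal color range.
# The graph has no textures or animation, and the color is its only parameter.
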Material /Game/Missions/Secondary/GatedEntrance/TechDoor/M_TechDoor_Button
graph {
 properties {pixelDepthOffsetMode=PDOM_Legacy, usedShadingModels=Default Lit}
 nodes [Multiply(A=VectorParameter() {defaultValue=<rgba 0 0.5613 1 0>}.RGB, B=Constant() {r=3})]
}
