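# M_TechDoor_LightGrey is a default-lit material, named as a light grey surface for the tech door in the gated entrance secondary mission.
# The graph holds a single scalar Constant node at mid value, with no textures or parameters.
# The file shows no connection from that constant to any material output.
# Pixel depth offset is set to legacy mode.
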Material /Game/Missions/Secondary/GatedEntrance/TechDoor/M_TechDoor_LightGrey
graph {properties {pixelDepthOffsetMode=PDOM_Legacy, usedShadingModels=Default Lit}, nodes [Constant() {r=0.5}]}
